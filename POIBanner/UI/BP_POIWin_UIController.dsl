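; BP_POIWin_UIController is a UI controller for the POI win banner, judging by its name and folder.
; Its only content is an UnbindHandles function that clears every listener from the OnUpdateVisibility and OnUpdateProgress event dispatchers.
; No binding, broadcasting or display logic for visibility or progress appears here.
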(fn UnbindHandles
  (Default|UnbindallEventsfromOnUpdateVisibility)
  (Default|UnbindallEventsfromOnUpdateProgress))
