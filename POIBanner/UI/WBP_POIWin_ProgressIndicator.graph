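# HUD indicator for POI claim progress, wrapping WBP_POIWin_ProgressWidget in non-static location mode and forwarding screen-clamp direction changes to it.
# On construct, it finds BP_POIWin_UIController on the player controller and binds its visibility, progress, entered, contested and claimed events.
# Regular visibility updates are inverted, collapsing when HitTestInvisible arrives, while forced updates apply directly.
# A claim collapses the widget and unbinds handlers until POIClaimResetAllProgress re-registers them.
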WidgetBlueprint /POIBanner/UI/WBP_POIWin_ProgressIndicator
graphs [
 BindToUIControllerEvents {nodes [
   n1 = BindToUIControllerEvents!() -> (then, BindToEvents:bool=~) {is=title K2Node_FunctionEntry}
   > n2 = Branch(Condition=@n1)
   n3 = Create Event()
   n4 = Get UIController() -> (UIController:BP POIWin UIController Object Reference)
   > Bind Event to On Update Visibility(self:BP POIWin UIController Object Reference=@n4, Delegate=@n3) {is=title K2Node_AddDelegate}
   n5 = Create Event()
   n6 = Get UIController()
   > Bind Event to Force Update Visibility(self:BP POIWin UIController Object Reference=@n6, Delegate=@n5) {is=title K2Node_AddDelegate}
   n7 = Create Event()
   n8 = Get UIController()
   > Bind Event to On Update Progress(self:BP POIWin UIController Object Reference=@n8, Delegate=@n7) {is=title K2Node_AddDelegate}
   n9 = Create Event(self=GetProgressWidget!() -> (MainProgressWidget:WBP POIWin Progress Widget Object Reference))
   n10 = Get UIController()
   > Bind Event to On POIContested(self:BP POIWin UIController Object Reference=@n10, Delegate=@n9) {is=title K2Node_AddDelegate}
   n11 = Create Event()
   n12 = Get UIController()
   > Bind Event to On POIClaimed(self:BP POIWin UIController Object Reference=@n12, Delegate=@n11) {is=title K2Node_AddDelegate}
   n13 = Create Event(self=GetProgressWidget())
   n14 = Get UIController()
   > Bind Event to On POIEntered(self:BP POIWin UIController Object Reference=@n14, Delegate=@n13) {is=title K2Node_AddDelegate}
   Unbind Event from On Update Visibility(execute=@n2.else, self:BP POIWin UIController Object Reference=@n4, Delegate=@n3) {is=title K2Node_RemoveDelegate}
   > Unbind Event from Force Update Visibility(self:BP POIWin UIController Object Reference=@n6, Delegate=@n5) {is=title K2Node_RemoveDelegate}
   > Unbind Event from On Update Progress(self:BP POIWin UIController Object Reference=@n8, Delegate=@n7) {is=title K2Node_RemoveDelegate}
   > Unbind Event from On POIContested(self:BP POIWin UIController Object Reference=@n10, Delegate=@n9) {is=title K2Node_RemoveDelegate}
   > Unbind Event from On POIClaimed(self:BP POIWin UIController Object Reference=@n12, Delegate=@n11) {is=title K2Node_RemoveDelegate}
   > Unbind Event from On POIEntered(self:BP POIWin UIController Object Reference=@n14, Delegate=@n13) {is=title K2Node_RemoveDelegate}
  ]}
 EventGraph {nodes [
   n1 = POIClaimResetAllProgress() {is=title K2Node_CustomEvent}
   RegisterHandlers(execute=[@n1.then, @n4.then])
   n2 = GetProgressWidget()
   > ResetClaimedState(self:WBP POIWin Progress Widget Object Reference=@n2)
   > Update Widget Location Type!(execute, self:WBP POIWin Progress Widget Object Reference=@n2, InIsStaticWidget:bool=false) -> (then)
   > SetVisibility(self=GetProgressWidget(), InVisibility=Collapsed)
   RegisterHandlers() {is=title K2Node_CustomEvent}
   > n3 = GetUIController!(execute, self) -> (then, UIController:BP POIWin UIController Object Reference)
   > Set UIController(UIController:BP POIWin UIController Object Reference=@n3) -> (Output_Get:BP POIWin UIController Object Reference)
   > BindToUIControllerEvents!(execute, self, BindToEvents:bool=true) -> (then) {is=title K2Node_CallFunction}
   UnbindHandlers() {is=title K2Node_CustomEvent}
   > BindToUIControllerEvents(BindToEvents=false)
   n4 = Event Construct()
  ]}
 GetProgressWidget {nodes [
   GetProgressWidget() {is=title K2Node_FunctionEntry}
   > Return Node!(
    execute
    MainProgressWidget:WBP POIWin Progress Widget Object Reference=Get ProgressWidget() -> (ProgressWidget:WBP POIWin Progress Widget Object Reference)
   )
  ]}
 GetUIController {nodes [
   GetUIController() {is=title K2Node_FunctionEntry}
   n1 = GetPlayerController!(PlayerIndex:int=0) -> (ReturnValue:Player Controller Object Reference)
   > Is Valid(InputObject=@n1)
   > Return Node!(
    execute
    UIController:BP POIWin UIController Object Reference=GetComponentByClass!(
     self:Actor Object Reference=@n1
     ComponentClass:Actor Component Class Reference=&/BP_POIWin_UIController.BP_POIWin_UIController_C
    ) -> (ReturnValue:BP POIWin UIController Object Reference)
   )
  ]}
 On Clamp State Changed {nodes [
   n1 = On Clamp State Changed!() -> (then, ClampDirection:EClampToScreenDirection Enum=Left) {is=title K2Node_FunctionEntry}
   > OnClampStateChanged!(
    execute
    self:WBP POIWin Progress Widget Object Reference=GetProgressWidget() {is=title K2Node_CallFunction}
    ClampToScreenDirection:EClampToScreenDirection Enum=@n1
   ) -> (then)
  ]}
 OnForceIndicatorVisibility {nodes [
   n1 = OnForceIndicatorVisibility!() -> (then, InVisibility:ESlateVisibility Enum=Visible, InCurrentProgress:double=~) {is=title K2Node_FunctionEntry}
   n2 = GetProgressWidget()
   > ResetClaimedState(self=@n2)
   > SetVisibility(self=@n2, InVisibility=@n1)
  ]}
 OnPOIClaimed {nodes [
   OnPOIClaimed!() -> (then, Dummy:bool=~) {is=title K2Node_FunctionEntry}
   n1 = GetProgressWidget()
   > ResetClaimedState(self=@n1)
   > SetVisibility(self=@n1, InVisibility=Collapsed)
   > UnbindHandlers() {is=title K2Node_CallFunction}
  ]}
 OnUpdateProgress {nodes [
   n1 = OnUpdateProgress!() -> (then, IsFriendly:bool=~, CurrentProgress:double=~) {is=title K2Node_FunctionEntry}
   > n2 = Set CachedCurrentProgress(CachedCurrentProgress:double=@n1.CurrentProgress) -> (Output_Get:double)
   > Update Claim Progress!(execute, self:WBP POIWin Progress Widget Object Reference=GetProgressWidget(), IsFriendlyTeam:bool=@n1, CurrentProgress:double=@n2) -> (then)
  ]}
 OnUpdateVisibility {nodes [
   n1 = OnUpdateVisibility!() -> (then, InVisibility:ESlateVisibility Enum=Visible, InCurrentProgress:double=~) {is=title K2Node_FunctionEntry}
   > SetVisibility(
    self=GetProgressWidget()
    InVisibility=Select(Option 0=HitTestInvisible, Option 1=Collapsed, Index="Equal (Enum)"(A:ESlateVisibility Enum=@n1, B:ESlateVisibility Enum=HitTestInvisible))
   )
  ]}
 ShowClaimOutcome {nodes [
   n1 = ShowClaimOutcome!() -> (then, Won:bool=~) {is=title K2Node_FunctionEntry}
   > ShowClaimOutcome!(execute, self:WBP POIWin Progress Widget Object Reference=GetProgressWidget(), Won:bool=@n1) -> (then) {is=title K2Node_CallFunction}
  ]}
]
variables [
 CachedCurrentProgress {type=number}
 UIController {type=BP_POIWin_UIController_C, default=null}
]
widgets {
 info {inheritedWidgetCount=1, parentClass=&FortHUDElementWidget, rootWidgetClass=&Overlay, widgetCount=4}
 widgets [
  Overlay_Root {widgetClassPath=&Overlay, widget=Overlay_Root}
  CommonVisibilitySwitcher {
   widgetClassPath=&CommonVisibilitySwitcher
   bIsVariable=true
   parent=Overlay_Root
   widget=CommonVisibilitySwitcher
   slot {class=OverlaySlot}
  }
  ProgressWidget {
   widgetClassPath=&/WBP_POIWin_ProgressWidget
   bIsVariable=true
   parent=CommonVisibilitySwitcher
   widget=ProgressWidget
   properties {elementTag {tagName=POIBanner.HUD.HideProgressIndicator}}
   slot {class=CommonVisibilitySwitcherSlot}
  }
  TutorialHighlight {widgetClassPath=&TutorialHighlight}
 ]
}
