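# HUD layout widget that shows point-of-interest claim progress and outcome through an embedded progress widget.
# It binds to the player controller's UI controller component, updating progress itself and routing contested and entered events to the progress widget.
# On claim, it plays a win pulse or lose shake ending in a fade, reverses it after a delay, resets, then broadcasts completion through the global event router.
# On construct, touch input offsets the main overlay.
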WidgetBlueprint /POIBanner/UI/WBP_POIWin_ProgressLayout
animations [
 Anim_Win {
  binding_count=1
  length_seconds=0.8
  object=&self:Anim_Win
  sequence {
   bindings [MainProgressWidget {
     id=#1
     possessed_class=WBP_POIWin_ProgressWidget_C
     tracks [
      {
       type=MovieScene2DTransformTrack
       sections [{
         type=MovieScene2DTransformSection
         end_frame=20
         start_frame=0
         channels [
          Translation.X {type=MovieSceneScriptingFloatChannel, const [0,1]}
          Translation.Y {type=MovieSceneScriptingFloatChannel, const [0,1]}
          Angle {type=MovieSceneScriptingFloatChannel, const [0,1]}
          Scale.X {type=MovieSceneScriptingFloatChannel, keys [[0,1],[8,1.2],[16,1]]}
          Scale.Y {type=MovieSceneScriptingFloatChannel, keys [[0,1],[8,1.2],[16,1]]}
          Shear.X {type=MovieSceneScriptingFloatChannel, const [0,1]}
          Shear.Y {type=MovieSceneScriptingFloatChannel, const [0,1]}
         ]
        }]
      }
      {
       type=MovieSceneFloatTrack
       display_name=Render Opacity
       sections [{type=MovieSceneFloatSection, end_frame=24, start_frame=0, channels [None {type=MovieSceneScriptingFloatChannel, keys [[0,1],[16,1],[24,0]]}]}]
      }
     ]
    }]
   display_rate {numerator=30}
  }
 }
 Anim_Lose {
  binding_count=1
  length_seconds=0.8
  object=&self:Anim_Lose
  sequence {
   bindings [MainProgressWidget {
     id=#2
     possessed_class=WBP_POIWin_ProgressWidget_C
     tracks [
      {
       type=MovieScene2DTransformTrack
       sections [{
         type=MovieScene2DTransformSection
         end_frame=20
         start_frame=0
         channels [
          Translation.X {type=MovieSceneScriptingFloatChannel, keys [[0,0],[8,-20],[12,20],[16,0]]}
          Translation.Y {type=MovieSceneScriptingFloatChannel, const [0,1]}
          Angle {type=MovieSceneScriptingFloatChannel, const [0,1]}
          Scale.X {type=MovieSceneScriptingFloatChannel, const [1,1]}
          Scale.Y {type=MovieSceneScriptingFloatChannel, const [1,1]}
          Shear.X {type=MovieSceneScriptingFloatChannel, const [0,1]}
          Shear.Y {type=MovieSceneScriptingFloatChannel, const [0,1]}
         ]
        }]
      }
      {
       type=MovieSceneFloatTrack
       display_name=Render Opacity
       sections [{type=MovieSceneFloatSection, end_frame=24, start_frame=0, channels [None {type=MovieSceneScriptingFloatChannel, keys [[0,1],[16,1],[24,0]]}]}]
      }
     ]
    }]
   display_rate {numerator=30}
  }
 }
]
graphs [
 BindToUIControllerEvents {nodes [
   n1 = BindToUIControllerEvents!() -> (then, BindToEvents:bool=~) {is=title K2Node_FunctionEntry}
   > n2 = Branch(Condition=@n1)
   n3 = Create Event()
   n4 = Get UIController() -> (UIController:BP POIWin UIController Object Reference)
   > Bind Event to On Update Progress(self:BP POIWin UIController Object Reference=@n4, Delegate=@n3) {is=title K2Node_AddDelegate}
   n5 = Create Event()
   n6 = Get UIController()
   > Bind Event to On POIClaimed(self:BP POIWin UIController Object Reference=@n6, Delegate=@n5) {is=title K2Node_AddDelegate}
   n7 = Create Event(self=GetProgressWidget!() -> (MainProgressWidget:WBP POIWin Progress Widget Object Reference))
   n8 = Get UIController()
   > Bind Event to On POIContested(self:BP POIWin UIController Object Reference=@n8, Delegate=@n7) {is=title K2Node_AddDelegate}
   n9 = Create Event(self=GetProgressWidget())
   n10 = Get UIController()
   > Bind Event to On POIEntered(self:BP POIWin UIController Object Reference=@n10, Delegate=@n9) {is=title K2Node_AddDelegate}
   Unbind Event from On Update Progress(execute=@n2.else, self:BP POIWin UIController Object Reference=@n4, Delegate=@n3) {is=title K2Node_RemoveDelegate}
   > Unbind Event from On POIClaimed(self:BP POIWin UIController Object Reference=@n6, Delegate=@n5) {is=title K2Node_RemoveDelegate}
   > Unbind Event from On POIContested(self:BP POIWin UIController Object Reference=@n8, Delegate=@n7) {is=title K2Node_RemoveDelegate}
   > Unbind Event from On POIEntered(self:BP POIWin UIController Object Reference=@n10, Delegate=@n9) {is=title K2Node_RemoveDelegate}
  ]}
 EventGraph {nodes [
   n1 = OnPOIClaimed!() -> (OutputDelegate, then, Won:bool=~) {is=title K2Node_CustomEvent}
   > ShowClaimOutcome!(execute, self:WBP POIWin Progress Widget Object Reference=GetProgressWidget(), Won:bool=@n1.Won) -> (then)
   n2 = Select(
    Option 0:Widget Animation Object Reference=Get Anim_Lose() -> (Anim_Lose:Widget Animation Object Reference)
    Option 1:Widget Animation Object Reference=Get Anim_Win() -> (Anim_Win:Widget Animation Object Reference)
    Index=@n1.Won
   ) -> (ReturnValue:Widget Animation Object Reference)
   > n3 = "Play Animation with Finished event (legacy)"!(
    execute:Exec
    Widget:User Widget Object Reference=Self-Reference()
    InAnimation:Widget Animation Object Reference=@n2
    StartAtTime:float=0
    NumLoopsToPlay:int=1
    PlayMode:EUMGSequencePlayMode Enum
    PlaybackSpeed:float=1
   ) -> (then:Exec, Result:UMGSequence Player Object Reference, Finished:Exec) {is=title K2Node_PlayAnimation}
   Delay(execute=@n3.Finished, Duration=Get WaitTimeBeforeOutAnim() -> (WaitTimeBeforeOutAnim:double))
   > n4 = "Play Animation with Finished event (legacy)"(Widget=Self-Reference(), InAnimation=@n2, StartAtTime=0, NumLoopsToPlay=1, PlayMode=Reverse, PlaybackSpeed=1)
   ResetClaimedState(execute=@n4.Finished, self:WBP POIWin Progress Widget Object Reference=GetProgressWidget())
   > Broadcast Event!(
    execute
    self:Gameplay Event Router Component Object Reference=GetGlobalEventRouter()
    EventData:POIResult Anim Finished Structure=Make POIResult Anim Finished!(bFinished_21_#3:bool=true) -> (POIResultAnimFinished:POIResult Anim Finished Structure) {is=title K2Node_MakeStruct}
    EventContext:Object Reference=GetOwningPlayer()
   ) -> (then, ReturnValue:bool)
   RegisterHandlers() {is=title K2Node_CustomEvent}
   > n5 = GetUIController!(execute, self) -> (then, UIController:BP POIWin UIController Object Reference)
   > Set UIController(UIController:BP POIWin UIController Object Reference=@n5) -> (Output_Get:BP POIWin UIController Object Reference)
   > BindToUIControllerEvents!(execute, self, BindToEvents:bool=true) -> (then) {is=title K2Node_CallFunction}
   Event Destruct()
   > BindToUIControllerEvents(BindToEvents=false)
   Event Construct()
   > RegisterHandlers() {is=title K2Node_CallFunction}
   n6 = GetProgressWidget()
   > ResetClaimedState(self=@n6)
   > SetVisibility(self=@n6, InVisibility=HitTestInvisible)
   > Update Widget Location Type!(execute, self:WBP POIWin Progress Widget Object Reference=@n6, InIsStaticWidget:bool=true) -> (then)
   > n7 = Branch(Condition=IsUsingTouch(PlayerController=GetOwningPlayer()))
   n8 = Get Overlay_Main() -> (Overlay_Main:Overlay Object Reference)
   > SetRenderTranslation(self=@n8, Translation=(X=110,Y=-120))
   SetRenderTranslation(execute=@n7.else, self=@n8, Translation=(X=0,Y=0))
  ]}
 GetProgressWidget {nodes [
   GetProgressWidget() {is=title K2Node_FunctionEntry}
   > Return Node!(
    execute
    MainProgressWidget:WBP POIWin Progress Widget Object Reference=Get MainProgressWidget() -> (MainProgressWidget:WBP POIWin Progress Widget Object Reference)
   )
  ]}
 GetUIController {nodes [
   GetUIController() {is=title K2Node_FunctionEntry}
   n1 = GetPlayerController!(PlayerIndex:int=0) -> (ReturnValue:Player Controller Object Reference)
   > Is Valid(InputObject=@n1)
   > Return Node!(
    execute
    UIController:BP POIWin UIController Object Reference=GetComponentByClass!(
     self:Actor Object Reference=@n1
     ComponentClass:Actor Component Class Reference=&/BP_POIWin_UIController.BP_POIWin_UIController_C
    ) -> (ReturnValue:BP POIWin UIController Object Reference)
   )
  ]}
 OnDisplayFinished {nodes [
   OnDisplayFinished() {is=title K2Node_FunctionEntry}
   > SetVisibility(self=GetProgressWidget() {is=title K2Node_CallFunction}, InVisibility=Hidden)
  ]}
 OnOutroAnimFinished {nodes [OnOutroAnimFinished() {is=title K2Node_FunctionEntry}]}
 OnUpdateProgress {nodes [
   n1 = OnUpdateProgress!() -> (then, IsFriendly:bool=~, CurrentProgress:double=~) {is=title K2Node_FunctionEntry}
   > n2 = Set Cached Current Progress(Cached Current Progress:double=@n1.CurrentProgress) -> (Output_Get:double)
   > Update Claim Progress!(
    execute
    self:WBP POIWin Progress Widget Object Reference=GetProgressWidget()
    IsFriendlyTeam:bool=Get IsFriendly() -> (IsFriendly:bool)
    CurrentProgress:double=@n2
   ) -> (then)
  ]}
 OnUpdateVisibility {nodes [
   OnUpdateVisibility!() -> (then, InVisibility:ESlateVisibility Enum=Visible, InCurrentProgress:double=~) {is=title K2Node_FunctionEntry}
   > n1 = Switch on ESlateVisibility!(execute, Selection:ESlateVisibility Enum=Get InVisibility() -> (InVisibility:ESlateVisibility Enum=Visible)) -> (Visible:Exec, Collapsed:Exec, Hidden:Exec, HitTestInvisible:Exec, SelfHitTestInvisible:Exec) {is=title K2Node_SwitchEnum}
   SetVisibility(execute=[@n1.Collapsed, @n1.Hidden], InVisibility=Collapsed)
   > Set ShowOnNextUpdate(ShowOnNextUpdate:bool=false) -> (Output_Get:bool)
   Set ShowOnNextUpdate(execute=[@n1.HitTestInvisible, @n1.SelfHitTestInvisible, @n1.Visible], ShowOnNextUpdate=true)
   > ResetClaimedState(self=GetProgressWidget())
   > SetVisibility(InVisibility=HitTestInvisible)
  ]}
]
variables [
 Cached Current Progress {type=number}
 OnOutroAnimFinished {}
 ShowOnNextUpdate {type=boolean}
 UIController {type=BP_POIWin_UIController_C, default=null}
 WaitTimeBeforeOutAnim {type=number, default=1}
]
widgets {
 info {inheritedWidgetCount=1, parentClass=&FortHUDElementWidget, rootWidgetClass=&Overlay, widgetCount=6}
 widgets [
  Overlay_Root {widgetClassPath=&Overlay, widget=Overlay_Root}
  Overlay_Main {
   widgetClassPath=&Overlay
   bIsVariable=true
   parent=Overlay_Root
   widget=Overlay_Main
   slot {class=OverlaySlot, properties {horizontalAlignment=HAlign_Center, verticalAlignment=VAlign_Bottom}}
  }
  SizeBox_0 {
   widgetClassPath=&SizeBox
   parent=Overlay_Main
   widget=SizeBox_0
   properties {bOverride_HeightOverride=true, heightOverride=250}
   slot {class=OverlaySlot, properties {horizontalAlignment=HAlign_Fill, verticalAlignment=VAlign_Fill}}
  }
  CommonVisibilitySwitcher {
   widgetClassPath=&CommonVisibilitySwitcher
   bIsVariable=true
   parent=SizeBox_0
   widget=CommonVisibilitySwitcher
   slot {class=SizeBoxSlot, properties {horizontalAlignment=HAlign_Center, verticalAlignment=VAlign_Top}}
  }
  MainProgressWidget {
   widgetClassPath=&/WBP_POIWin_ProgressWidget
   bIsVariable=true
   parent=CommonVisibilitySwitcher
   widget=MainProgressWidget
   properties {elementTag {tagName=POIBanner.HUD.HideProgressIndicator}}
   slot {class=CommonVisibilitySwitcherSlot}
  }
  TutorialHighlight {widgetClassPath=&TutorialHighlight}
 ]
}
